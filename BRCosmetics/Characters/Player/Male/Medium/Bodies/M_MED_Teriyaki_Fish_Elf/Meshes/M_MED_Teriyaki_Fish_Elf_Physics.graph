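# Physics asset for the medium male Teriyaki Fish Elf player body, setting up secondary motion for its bags and upper and lower cloth pieces.
# Capsules cover a few body bones (pelvis, upper spine, clavicles, thighs, right upper arm) and every dangling piece.
# Each piece is a two-link chain with limited swing and twist: a base body constrained to pelvis, thigh, spine_05 or clavicle, then the hanging body constrained to that base.
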PhysicsAsset /BRCosmetics/Characters/Player/Male/Medium/Bodies/M_MED_Teriyaki_Fish_Elf/Meshes/M_MED_Teriyaki_Fish_Elf_Physics
bodies {
 clavicle_l [{shapeType=Capsule, length=14.0677, radius=6.2883, center=<13.2136 1.5078 -2.2054>, rotation=<rot -0.1688 96.9815 -115.0578>}]
 clavicle_r [{shapeType=Capsule, length=11.2183, radius=6.3521, center=<-11.2941 -1.9831 1.7426>, rotation=<rot -1.4804 -80.6695 -71.4374>}]
 dyn_bag01 [{shapeType=Capsule, length=3, radius=3, center=<1.541 0.0001 -3.3145>, rotation=<rot 20.0587 0 0>}]
 dyn_bag02 [{shapeType=Capsule, length=3.25, radius=3.25, center=<1.2836 0 -2.4057>, rotation=<rot 12.2233 0 0>}]
 dyn_bag03 [{shapeType=Capsule, length=5.5, radius=2.5, center=<1.0689 0 -3.8914>, rotation=<rot 11.4279 0 0>}]
 dyn_bagBase01 [{shapeType=Capsule, length=0.2501, radius=1.5731, rotation=<rot 0 -90 0>}]
 dyn_bagBase02 [{shapeType=Capsule, length=1.2678, radius=1.5731, rotation=<rot 0 -90 0>}]
 dyn_bagBase03 [{shapeType=Capsule, length=0, radius=1.3844, rotation=<rot 0 -90 0>}]
 dyn_cloth_dn01 [{shapeType=Capsule, length=3.25, radius=3, center=<-0.2555 0.0818 -2.9474>, rotation=<rot -30.9375 0 -0.0002>}]
 dyn_cloth_dn02 [{shapeType=Capsule, length=3.25, radius=3, center=<0 0 -1.2565>}]
 dyn_cloth_dn03 [{shapeType=Capsule, length=3.25, radius=3, center=<0 0 -1.2565>}]
 dyn_cloth_dn04 [{shapeType=Capsule, length=3.25, radius=3, center=<0.7623 -0.0001 -2.5985>}]
 dyn_cloth_dn05 [{shapeType=Capsule, length=3.25, radius=3, center=<0 0 -2.672>}]
 dyn_cloth_dn06 [{shapeType=Capsule, length=3.25, radius=3, center=<0 0 -1.2565>}]
 dyn_cloth_dn07 [{shapeType=Capsule, length=5.25, radius=3, center=<0 0.329 -2.9546>, rotation=<rot -0.0001 0 -90.0001>}]
 dyn_cloth_dnBase01 [{shapeType=Capsule, length=0.505, radius=0.505, rotation=<rot 0 -90 0>}]
 dyn_cloth_dnBase02 [{shapeType=Capsule, length=0.505, radius=0.505, rotation=<rot 0 -90 0>}]
 dyn_cloth_dnBase03 [{shapeType=Capsule, length=3.25, radius=3, center=<0 0 -1.2565>}]
 dyn_cloth_dnBase04 [{shapeType=Capsule, length=3.25, radius=3, center=<0.7623 -0.0001 -0.9445>}]
 dyn_cloth_dnBase05 [{shapeType=Capsule, length=0.505, radius=0.505, rotation=<rot 0 -90 0>}]
 dyn_cloth_dnBase06 [{shapeType=Capsule, length=0.505, radius=0.505, rotation=<rot 0 -90 0>}]
 dyn_cloth_dnBase07 [{shapeType=Capsule, length=0.505, radius=0.505, rotation=<rot 0 -90 0>}]
 dyn_cloth_upBase_l_01 [{shapeType=Capsule, length=0.505, radius=0.505, rotation=<rot 0 -90 0>}]
 dyn_cloth_upBase_l_02 [{shapeType=Capsule, length=0.505, radius=0.505, rotation=<rot 0 -90 0>}]
 dyn_cloth_upBase_l_03 [{shapeType=Capsule, length=0.505, radius=0.505, rotation=<rot 0 -90 0>}]
 dyn_cloth_up_l_01 [{shapeType=Capsule, length=4, radius=3, center=<0.3525 0 -3.0586>}]
 dyn_cloth_up_l_02 [{shapeType=Capsule, length=4.75, radius=3, center=<0.2248 0.0018 -2.711>, rotation=<rot -0.7445 -0.2242 0.7961>}]
 dyn_cloth_up_l_03 [{shapeType=Capsule, length=4, radius=2.5, center=<0.3525 0 -3.0586>}]
 dyn_croth_upBase_r_01 [{shapeType=Capsule, length=0.505, radius=0.505, rotation=<rot 0 -90 0>}]
 dyn_croth_upBase_r_02 [{shapeType=Capsule, length=0.505, radius=0.505, rotation=<rot 0 -90 0>}]
 dyn_croth_upBase_r_03 [{shapeType=Capsule, length=0.505, radius=0.505, rotation=<rot 0 -90 0>}]
 dyn_croth_up_r_01 [{shapeType=Capsule, length=4, radius=3, center=<0.3525 0 -3.0586>}]
 dyn_croth_up_r_02 [{shapeType=Capsule, length=4.75, radius=3, center=<0.3438 -0.0001 -2.706>, rotation=<rot -0.0554 0 0>}]
 dyn_croth_up_r_03 [{shapeType=Capsule, length=4, radius=2.5, center=<0.3525 0 -3.0586>}]
 pelvis [{shapeType=Capsule, length=16.0289, radius=10.6291, center=<-5.4759 -0.0646 0>, rotation=<rot 0 -91.2371 0.6784>}]
 spine_05 [
  {shapeType=Capsule, length=13.5245, radius=13.0805, center=<-1.6783 -0.9437 0>}
  {shapeType=Capsule, length=13.5245, radius=13.0805, center=<-5.1329 1.7631 0>}
 ]
 thigh_twist_01_l [{shapeType=Capsule, length=26.888, radius=10.0154, center=<6.0825 0.0004 -0.8701>, rotation=<rot 89.2972 0 0>}]
 thigh_twist_01_r [{shapeType=Capsule, length=26.888, radius=10.0154, center=<-5.0456 0.0004 2.6224>, rotation=<rot 89.2972 0 0>}]
 upperarm_r [{shapeType=Capsule, length=17.7322, radius=5.3313, center=<-11.5619 -0.5722 0.3584>, rotation=<rot -83.17 -44.6078 45.8914>}]
}
constraints [
 {bone1Name=dyn_cloth_dn07, bone2Name=dyn_cloth_dnBase07, swing1LimitDegrees=10, swing2LimitDegrees=40, twistLimitDegrees=10}
 {bone1Name=dyn_cloth_dn05, bone2Name=dyn_cloth_dnBase05, swing1LimitDegrees=10, swing2LimitDegrees=40, twistLimitDegrees=10}
 {bone1Name=dyn_bag01, bone2Name=dyn_bagBase01, swing1LimitDegrees=5, swing2LimitDegrees=5, twistLimitDegrees=15}
 {bone1Name=dyn_bag02, bone2Name=dyn_bagBase02, swing1LimitDegrees=5, swing2LimitDegrees=5, twistLimitDegrees=15}
 {bone1Name=dyn_bag03, bone2Name=dyn_bagBase03, swing1LimitDegrees=5, swing2LimitDegrees=10, twistLimitDegrees=10}
 {bone1Name=dyn_cloth_dn06, bone2Name=dyn_cloth_dnBase06, swing1LimitDegrees=10, swing2LimitDegrees=40, twistLimitDegrees=10}
 {bone1Name=dyn_cloth_up_l_01, bone2Name=dyn_cloth_upBase_l_01, swing1LimitDegrees=20, swing2LimitDegrees=20, twistLimitDegrees=20}
 {bone1Name=dyn_cloth_up_l_03, bone2Name=dyn_cloth_upBase_l_03, swing1LimitDegrees=20, swing2LimitDegrees=20, twistLimitDegrees=20}
 {bone1Name=dyn_croth_up_r_01, bone2Name=dyn_croth_upBase_r_01, swing1LimitDegrees=20, swing2LimitDegrees=20, twistLimitDegrees=20}
 {bone1Name=dyn_croth_up_r_03, bone2Name=dyn_croth_upBase_r_03, swing1LimitDegrees=20, swing2LimitDegrees=20, twistLimitDegrees=20}
 {bone1Name=dyn_croth_up_r_02, bone2Name=dyn_croth_upBase_r_02, swing1LimitDegrees=20, swing2LimitDegrees=20, twistLimitDegrees=20}
 {bone1Name=dyn_bagBase03, bone2Name=pelvis, swing1LimitDegrees=5, swing2LimitDegrees=5, twistLimitDegrees=5}
 {bone1Name=dyn_bagBase02, bone2Name=pelvis, swing1LimitDegrees=5, swing2LimitDegrees=5, twistLimitDegrees=5}
 {bone1Name=dyn_bagBase01, bone2Name=pelvis, swing1LimitDegrees=20, swing2LimitDegrees=20, twistLimitDegrees=20}
 {bone1Name=dyn_cloth_dnBase07, bone2Name=pelvis, swing1LimitDegrees=20, swing2LimitDegrees=20, twistLimitDegrees=20}
 {bone1Name=dyn_cloth_dnBase06, bone2Name=pelvis, swing1LimitDegrees=20, swing2LimitDegrees=20, twistLimitDegrees=20}
 {bone1Name=dyn_cloth_dnBase05, bone2Name=pelvis, swing1LimitDegrees=20, swing2LimitDegrees=20, twistLimitDegrees=20}
 {bone1Name=dyn_croth_upBase_r_03, bone2Name=spine_05, swing1LimitDegrees=20, swing2LimitDegrees=20, twistLimitDegrees=20}
 {bone1Name=dyn_croth_upBase_r_01, bone2Name=spine_05, swing1LimitDegrees=20, swing2LimitDegrees=20, twistLimitDegrees=20}
 {bone1Name=dyn_cloth_upBase_l_03, bone2Name=spine_05, swing1LimitDegrees=20, swing2LimitDegrees=20, twistLimitDegrees=20}
 {bone1Name=dyn_cloth_upBase_l_01, bone2Name=spine_05, swing1LimitDegrees=20, swing2LimitDegrees=20, twistLimitDegrees=20}
 {bone1Name=dyn_croth_upBase_r_02, bone2Name=clavicle_r, swing1LimitDegrees=20, swing2LimitDegrees=20, twistLimitDegrees=20}
 {bone1Name=dyn_cloth_upBase_l_02, bone2Name=clavicle_l, swing1LimitDegrees=20, swing2LimitDegrees=20, twistLimitDegrees=20}
 {bone1Name=dyn_cloth_up_l_02, bone2Name=dyn_cloth_upBase_l_02, swing1LimitDegrees=20, swing2LimitDegrees=20, twistLimitDegrees=20}
 {bone1Name=dyn_cloth_dnBase04, bone2Name=thigh_twist_01_r, swing1LimitDegrees=40, swing2LimitDegrees=40, twistLimitDegrees=40}
 {bone1Name=dyn_cloth_dn04, bone2Name=dyn_cloth_dnBase04, swing1LimitDegrees=40, swing2LimitDegrees=40, twistLimitDegrees=40}
 {bone1Name=dyn_cloth_dnBase03, bone2Name=thigh_twist_01_r, swing1LimitDegrees=5, swing2LimitDegrees=5, twistLimitDegrees=5}
 {bone1Name=dyn_cloth_dn03, bone2Name=dyn_cloth_dnBase03, swing1LimitDegrees=10, swing2LimitDegrees=30, twistLimitDegrees=10}
 {bone1Name=dyn_cloth_dnBase02, bone2Name=thigh_twist_01_l, swing1LimitDegrees=5, swing2LimitDegrees=5, twistLimitDegrees=5}
 {bone1Name=dyn_cloth_dn02, bone2Name=dyn_cloth_dnBase02, swing1LimitDegrees=10, swing2LimitDegrees=30, twistLimitDegrees=10}
 {bone1Name=dyn_cloth_dnBase01, bone2Name=thigh_twist_01_l, swing1LimitDegrees=5, swing2LimitDegrees=5, twistLimitDegrees=5}
 {bone1Name=dyn_cloth_dn01, bone2Name=dyn_cloth_dnBase01, swing1LimitDegrees=10, swing2LimitDegrees=10, twistLimitDegrees=10}
]
