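# Physics asset for the Teriyaki Fish Elf male medium body cosmetic, simulating secondary motion on its hat bones.
# The root hat capsule anchors a three-link hat chain that loosens slightly along its length, left and right three-link belt chains, and ten separate cloth capsules.
# A separate belt box carries two smaller boxes and has no constraint to the hat.
# All joints use tight angular limits, keeping the sway subtle.
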PhysicsAsset /BRCosmetics/Characters/Player/Male/Medium/Bodies/M_MED_Teriyaki_Fish_Elf/Meshes/Parts/M_MED_Teriyaki_Fish_Elf_Physics
bodies {
 dyn_hat [{shapeType=Capsule, length=0, radius=11.1327, center=<-0.7225 2.0418 -12.2882>, rotation=<rot 0 -90.0824 -89.2501>}]
 dyn_hat_01 [{shapeType=Capsule, length=2.0026, radius=5.8369, rotation=<rot 0 -90 0>}]
 dyn_hat_02 [{shapeType=Capsule, length=5.4621, radius=4.2906, center=<0.1101 -0.0699 1.4469>, rotation=<rot 0 -1.661 -1.7412>}]
 dyn_hat_03 [{shapeType=Capsule, length=0, radius=2.7369, center=<0.0524 -0.0438 2.517>, rotation=<rot -5.9548 -89.9996 0.3302>}]
 dyn_hat_belt [{shapeType=Box, extentX=4.1385, extentY=3.357, extentZ=2.0404, center=<0 0.4015 0.5696>}]
 dyn_hat_belt_01 [{shapeType=Box, extentX=2.129, extentY=1.0435, extentZ=2.9645, center=<0 0 -0.8323>}]
 dyn_hat_belt_02 [{shapeType=Box, extentX=2.129, extentY=1.0435, extentZ=2.9645, center=<0 0 -0.8323>}]
 dyn_hat_belt_l_01 [{shapeType=Capsule, length=2.7481, radius=3.0019}]
 dyn_hat_belt_l_02 [{shapeType=Capsule, length=3.3846, radius=2.3654}]
 dyn_hat_belt_l_03 [{shapeType=Capsule, length=3.75, radius=2, center=<0.0244 0 -1.4252>, rotation=<rot 0.9803 0 0>}]
 dyn_hat_belt_r_01 [{shapeType=Capsule, length=3, radius=2.75, rotation=<rot -57.3827 -0.0001 0>}]
 dyn_hat_belt_r_02 [{shapeType=Capsule, length=3.5, radius=2.25, rotation=<rot -55.8262 0 0>}]
 dyn_hat_belt_r_03 [{shapeType=Capsule, length=3.75, radius=2, rotation=<rot -55.6676 -0.0001 0.0001>}]
 dyn_hat_cloth_01 [{shapeType=Capsule, length=4.25, radius=2.25}]
 dyn_hat_cloth_02 [{shapeType=Capsule, length=4.25, radius=2.25}]
 dyn_hat_cloth_03 [{shapeType=Capsule, length=4.25, radius=2.25}]
 dyn_hat_cloth_04 [{shapeType=Capsule, length=4.25, radius=2.25}]
 dyn_hat_cloth_05 [{shapeType=Capsule, length=4.25, radius=2.25}]
 dyn_hat_cloth_06 [{shapeType=Capsule, length=4.25, radius=2.25}]
 dyn_hat_cloth_07 [{shapeType=Capsule, length=4.25, radius=2.25}]
 dyn_hat_cloth_08 [{shapeType=Capsule, length=4.25, radius=2.25}]
 dyn_hat_cloth_09 [{shapeType=Capsule, length=4.25, radius=2.25}]
 dyn_hat_cloth_10 [{shapeType=Capsule, length=4.25, radius=2.25}]
}
constraints [
 {bone1Name=dyn_hat_belt_01, bone2Name=dyn_hat_belt, swing1LimitDegrees=5, swing2LimitDegrees=5, twistLimitDegrees=5}
 {bone1Name=dyn_hat_belt_02, bone2Name=dyn_hat_belt, swing1LimitDegrees=5, swing2LimitDegrees=5, twistLimitDegrees=5}
 {bone1Name=dyn_hat_belt_l_01, bone2Name=dyn_hat, swing1LimitDegrees=10, swing2LimitDegrees=10, twistLimitDegrees=10}
 {bone1Name=dyn_hat_belt_l_02, bone2Name=dyn_hat_belt_l_01, swing1LimitDegrees=10, swing2LimitDegrees=10, twistLimitDegrees=10}
 {bone1Name=dyn_hat_belt_l_03, bone2Name=dyn_hat_belt_l_02, swing1LimitDegrees=10, swing2LimitDegrees=10, twistLimitDegrees=10}
 {bone1Name=dyn_hat_belt_r_01, bone2Name=dyn_hat, swing1LimitDegrees=10, swing2LimitDegrees=10, twistLimitDegrees=10}
 {bone1Name=dyn_hat_belt_r_02, bone2Name=dyn_hat_belt_r_01, swing1LimitDegrees=10, swing2LimitDegrees=10, twistLimitDegrees=10}
 {bone1Name=dyn_hat_belt_r_03, bone2Name=dyn_hat_belt_r_02, swing1LimitDegrees=10, swing2LimitDegrees=10, twistLimitDegrees=10}
 {bone1Name=dyn_hat_01, bone2Name=dyn_hat, swing1LimitDegrees=1, swing2LimitDegrees=1, twistLimitDegrees=1}
 {bone1Name=dyn_hat_02, bone2Name=dyn_hat_01, swing1LimitDegrees=2, swing2LimitDegrees=2, twistLimitDegrees=2}
 {bone1Name=dyn_hat_03, bone2Name=dyn_hat_02, swing1LimitDegrees=3, swing2LimitDegrees=3, twistLimitDegrees=3}
 {bone1Name=dyn_hat_cloth_01, bone2Name=dyn_hat, swing1LimitDegrees=10, swing2LimitDegrees=10, twistLimitDegrees=10}
 {bone1Name=dyn_hat_cloth_02, bone2Name=dyn_hat, swing1LimitDegrees=10, swing2LimitDegrees=10, twistLimitDegrees=10}
 {bone1Name=dyn_hat_cloth_03, bone2Name=dyn_hat, swing1LimitDegrees=10, swing2LimitDegrees=10, twistLimitDegrees=10}
 {bone1Name=dyn_hat_cloth_04, bone2Name=dyn_hat, swing1LimitDegrees=10, swing2LimitDegrees=10, twistLimitDegrees=10}
 {bone1Name=dyn_hat_cloth_05, bone2Name=dyn_hat, swing1LimitDegrees=10, swing2LimitDegrees=10, twistLimitDegrees=10}
 {bone1Name=dyn_hat_cloth_06, bone2Name=dyn_hat, swing1LimitDegrees=10, swing2LimitDegrees=10, twistLimitDegrees=10}
 {bone1Name=dyn_hat_cloth_07, bone2Name=dyn_hat, swing1LimitDegrees=10, swing2LimitDegrees=10, twistLimitDegrees=10}
 {bone1Name=dyn_hat_cloth_08, bone2Name=dyn_hat, swing1LimitDegrees=10, swing2LimitDegrees=10, twistLimitDegrees=10}
 {bone1Name=dyn_hat_cloth_09, bone2Name=dyn_hat, swing1LimitDegrees=10, swing2LimitDegrees=10, twistLimitDegrees=10}
 {bone1Name=dyn_hat_cloth_10, bone2Name=dyn_hat, swing1LimitDegrees=10, swing2LimitDegrees=10, twistLimitDegrees=10}
]
