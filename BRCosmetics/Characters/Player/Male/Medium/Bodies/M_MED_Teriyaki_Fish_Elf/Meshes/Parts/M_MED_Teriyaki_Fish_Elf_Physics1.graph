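# Physics asset for the hat of the Teriyaki Fish Elf male medium body cosmetic, covering its dyn_hat bones.
# Capsules form the hat chain, while boxes cover the belt pieces and ten cloth bones.
# The hat chain, two three-link side belt chains and individual cloth bodies branch from dyn_hat with tight swing and twist limits.
# Two small belt bodies hang from dyn_hat_belt, which has no constraint to dyn_hat.
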PhysicsAsset /BRCosmetics/Characters/Player/Male/Medium/Bodies/M_MED_Teriyaki_Fish_Elf/Meshes/Parts/M_MED_Teriyaki_Fish_Elf_Physics1
bodies {
 dyn_hat [{shapeType=Capsule, length=0, radius=11.1327, center=<-0.7225 2.0418 -12.2882>, rotation=<rot 0 -90.0824 -89.2501>}]
 dyn_hat_01 [{shapeType=Capsule, length=0.2628, radius=7.3305, rotation=<rot 0 -90 0>}]
 dyn_hat_02 [{shapeType=Capsule, length=4.3001, radius=4.9728, center=<0.1095 -0.0906 2.1268>, rotation=<rot 0 -1.661 -1.7412>}]
 dyn_hat_03 [{shapeType=Capsule, length=4.3763, radius=4.3723, center=<0.0524 -0.0438 2.517>, rotation=<rot -5.9548 -89.9996 0.3302>}]
 dyn_hat_belt [{shapeType=Box, extentX=4.1385, extentY=3.357, extentZ=2.0404, center=<0 0.4015 0.5696>}]
 dyn_hat_belt_01 [{shapeType=Box, extentX=2.129, extentY=1.0435, extentZ=2.9645, center=<0 0 -0.8323>}]
 dyn_hat_belt_02 [{shapeType=Box, extentX=2.129, extentY=1.0435, extentZ=2.9645, center=<0 0 -0.8323>}]
 dyn_hat_belt_l_01 [{shapeType=Box, extentX=2.4723, extentY=7.5, extentZ=5.0913, center=<0 0.0002 -1.3042>}]
 dyn_hat_belt_l_02 [{shapeType=Box, extentX=2.4723, extentY=7.5, extentZ=5.0913, center=<0 0.0002 -1.3042>}]
 dyn_hat_belt_l_03 [{shapeType=Box, extentX=2.4723, extentY=7.5, extentZ=5.0913, center=<0 0.0002 -1.3042>}]
 dyn_hat_belt_r_01 [{shapeType=Box, extentX=5.7511, extentY=7.5, extentZ=2.5589, center=<-0.1461 0 0.2089>, rotation=<rot 31.1198 0 0>}]
 dyn_hat_belt_r_02 [{shapeType=Box, extentX=5.7511, extentY=7.5, extentZ=2.5589, center=<-0.1461 0 0.2089>, rotation=<rot 31.1198 0 0>}]
 dyn_hat_belt_r_03 [{shapeType=Box, extentX=5.7511, extentY=7.5, extentZ=2.5589, center=<-0.1461 0 0.2089>, rotation=<rot 31.1196 0.0001 0.0001>}]
 dyn_hat_cloth_01 [{shapeType=Box, extentX=7.5, extentY=0.9482, extentZ=7.5}]
 dyn_hat_cloth_02 [{shapeType=Box, extentX=7.5, extentY=0.9482, extentZ=7.5}]
 dyn_hat_cloth_03 [{shapeType=Box, extentX=7.5, extentY=0.9482, extentZ=7.5}]
 dyn_hat_cloth_04 [{shapeType=Box, extentX=7.5, extentY=0.9482, extentZ=7.5}]
 dyn_hat_cloth_05 [{shapeType=Box, extentX=7.5, extentY=0.9482, extentZ=7.5}]
 dyn_hat_cloth_06 [{shapeType=Box, extentX=7.5, extentY=0.9482, extentZ=7.5}]
 dyn_hat_cloth_07 [{shapeType=Box, extentX=7.5, extentY=0.9482, extentZ=7.5}]
 dyn_hat_cloth_08 [{shapeType=Box, extentX=7.5, extentY=0.9482, extentZ=7.5}]
 dyn_hat_cloth_09 [{shapeType=Box, extentX=7.5, extentY=0.9482, extentZ=7.5}]
 dyn_hat_cloth_10 [{shapeType=Box, extentX=7.5, extentY=0.9482, extentZ=7.5}]
}
constraints [
 {bone1Name=dyn_hat_belt_01, bone2Name=dyn_hat_belt, swing1LimitDegrees=5, swing2LimitDegrees=5, twistLimitDegrees=5}
 {bone1Name=dyn_hat_belt_02, bone2Name=dyn_hat_belt, swing1LimitDegrees=5, swing2LimitDegrees=5, twistLimitDegrees=5}
 {bone1Name=dyn_hat_belt_l_01, bone2Name=dyn_hat, swing1LimitDegrees=10, swing2LimitDegrees=10, twistLimitDegrees=10}
 {bone1Name=dyn_hat_belt_l_02, bone2Name=dyn_hat_belt_l_01, swing1LimitDegrees=10, swing2LimitDegrees=10, twistLimitDegrees=10}
 {bone1Name=dyn_hat_belt_l_03, bone2Name=dyn_hat_belt_l_02, swing1LimitDegrees=10, swing2LimitDegrees=10, twistLimitDegrees=10}
 {bone1Name=dyn_hat_belt_r_01, bone2Name=dyn_hat, swing1LimitDegrees=10, swing2LimitDegrees=10, twistLimitDegrees=10}
 {bone1Name=dyn_hat_belt_r_02, bone2Name=dyn_hat_belt_r_01, swing1LimitDegrees=10, swing2LimitDegrees=10, twistLimitDegrees=10}
 {bone1Name=dyn_hat_belt_r_03, bone2Name=dyn_hat_belt_r_02, swing1LimitDegrees=10, swing2LimitDegrees=10, twistLimitDegrees=10}
 {bone1Name=dyn_hat_01, bone2Name=dyn_hat, swing1LimitDegrees=10, swing2LimitDegrees=10, twistLimitDegrees=10}
 {bone1Name=dyn_hat_02, bone2Name=dyn_hat_01, swing1LimitDegrees=10, swing2LimitDegrees=10, twistLimitDegrees=10}
 {bone1Name=dyn_hat_03, bone2Name=dyn_hat_02, swing1LimitDegrees=2, swing2LimitDegrees=2, twistLimitDegrees=2}
 {bone1Name=dyn_hat_cloth_01, bone2Name=dyn_hat, swing1LimitDegrees=10, swing2LimitDegrees=10, twistLimitDegrees=10}
 {bone1Name=dyn_hat_cloth_02, bone2Name=dyn_hat, swing1LimitDegrees=10, swing2LimitDegrees=10, twistLimitDegrees=10}
 {bone1Name=dyn_hat_cloth_03, bone2Name=dyn_hat, swing1LimitDegrees=10, swing2LimitDegrees=10, twistLimitDegrees=10}
 {bone1Name=dyn_hat_cloth_04, bone2Name=dyn_hat, swing1LimitDegrees=10, swing2LimitDegrees=10, twistLimitDegrees=10}
 {bone1Name=dyn_hat_cloth_05, bone2Name=dyn_hat, swing1LimitDegrees=10, swing2LimitDegrees=10, twistLimitDegrees=10}
 {bone1Name=dyn_hat_cloth_06, bone2Name=dyn_hat, swing1LimitDegrees=10, swing2LimitDegrees=10, twistLimitDegrees=10}
 {bone1Name=dyn_hat_cloth_07, bone2Name=dyn_hat, swing1LimitDegrees=10, swing2LimitDegrees=10, twistLimitDegrees=10}
 {bone1Name=dyn_hat_cloth_08, bone2Name=dyn_hat, swing1LimitDegrees=10, swing2LimitDegrees=10, twistLimitDegrees=10}
 {bone1Name=dyn_hat_cloth_09, bone2Name=dyn_hat, swing1LimitDegrees=10, swing2LimitDegrees=10, twistLimitDegrees=10}
 {bone1Name=dyn_hat_cloth_10, bone2Name=dyn_hat, swing1LimitDegrees=10, swing2LimitDegrees=10, twistLimitDegrees=10}
]
